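# Animation Blueprint for the DragonPry backpack, adding secondary motion to its left and right tassels and bow.
# Additive bone offsets set the tassel base, then chained Anim Dynamics nodes simulate each dangling bone with springs, limits and gravity.
# Bool-alpha additive rotations then re-pose the lower bow and tassels with timed blends.
# The event graph tracks melee-weapon, skydiving and DragonPry-pickaxe flags, though no visible pose link reads them.
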AnimBlueprint /BRCosmetics/Accessories/FORT_Backpacks/Backpack_M_MED_DragonPry/Meshes/M_MED_DragonPry_Pack_AnimBP
target_skeleton=&/M_MED_DragonPry_Pack_Skeleton
graphs [
 AnimGraph {nodes [
   n1 = "Transform (Modify) Bone - Bone: dyn_R_tassle_00"(ComponentPose=Local To Component(LocalPose=Input Pose()), Alpha=1, Translation="0,0,0", Rotation="-13.118,17.419,-0.269", Scale="1,1,1") {rotationMode=BMM_Additive, translationMode=BMM_Additive, boneToModify {boneName=dyn_R_tassle_00}, rotation=<rot -13.118 17.419 -0.269>}
   n2 = "Transform (Modify) Bone - Bone: dyn_L_tassle_00"(ComponentPose=@n1, Alpha=1, Translation="-0.131,-0.085,0.000", Rotation="9.477,16.963,0.847", Scale="1,1,1") {
    rotationMode=BMM_Additive
    translationMode=BMM_Additive
    boneToModify {boneName=dyn_L_tassle_00}
    rotation=<rot 9.477 16.963 0.847>
    translation=<-0.131 -0.085>
   }
   n3 = Anim Dynamics - Bone: dyn_L_tassle_00(ComponentPose=@n2, Alpha=1, ExternalForce="0,0,0") {
    angularDampingOverride=1
    angularSpringConstant=3000
    bAngularSpring=true
    bOverrideAngularDamping=true
    bUseGravityOverride=true
    boundBone {boneName=dyn_L_tassle_00}
    componentAppliedLinearAccClamp=<1000 1000 1600>
    componentLinearAccScale=<0.5 0.5 0.5>
    componentLinearVelScale=<0.6 0.6 0.6>
    gravityOverride {z=-700}
    physicsBodyDefinitions [{
      boundBone {boneName=dyn_L_tassle_00}
      boxExtents=<10 10 10>
      constraintSetup {angularLimitsMax=<30 0 40>, angularLimitsMin=<-30 -10 -15>, angularTarget=<0 0 0>}
      localJointOffset=<1.3 0 0>
     }]
    simSpaceSettings {maxAngularAcceleration=100, maxAngularVelocity=8, simSpaceAngularAlpha=0.2}
   }
   n4 = Anim Dynamics - Bone: dyn_L_tassle_01(ComponentPose=@n3, Alpha=1, ExternalForce="0,0,0") {
    angularDampingOverride=1
    angularSpringConstant=600
    bAngularSpring=true
    bOverrideAngularDamping=true
    bUseGravityOverride=true
    boundBone {boneName=dyn_L_tassle_01}
    componentAppliedLinearAccClamp=<800 800 1000>
    componentLinearAccScale=<0.4 0.4 0.4>
    componentLinearVelScale=<0.5 0.5 0.5>
    gravityOverride {z=-700}
    physicsBodyDefinitions [{
      boundBone {boneName=dyn_L_tassle_01}
      boxExtents=<10 10 10>
      constraintSetup {angularLimitsMax=<-50 30 15>, angularLimitsMin=<-50 -60 -10>, angularTarget=<1 0 0>}
      localJointOffset=<1.5 0.3 0>
     }]
    simSpaceSettings {maxAngularAcceleration=100, maxAngularVelocity=8, simSpaceAngularAlpha=0.2}
   }
   n5 = Anim Dynamics - Bone: dyn_L_tassle_02(ComponentPose=@n4, Alpha=1, ExternalForce="0,0,0") {
    angularDampingOverride=1
    angularSpringConstant=1000
    bAngularSpring=true
    bOverrideAngularDamping=true
    bUseGravityOverride=true
    boundBone {boneName=dyn_L_tassle_02}
    componentAppliedLinearAccClamp=<800 800 1000>
    componentLinearAccScale=<0.5 0.5 0.5>
    componentLinearVelScale=<0.6 0.6 0.6>
    gravityOverride {z=-500}
    physicsBodyDefinitions [{
      boundBone {boneName=dyn_L_tassle_02}
      boxExtents=<10 10 10>
      constraintSetup {angularLimitsMax=<0 40 20>, angularLimitsMin=<0 -60 -20>, angularTarget=<1 0 0>}
      localJointOffset=<2 0 0>
     }]
    simSpaceSettings {maxAngularAcceleration=100, maxAngularVelocity=10, simSpaceAngularAlpha=0.2}
   }
   n6 = Anim Dynamics - Bone: dyn_R_tassle_00(ComponentPose=@n5, Alpha=1, ExternalForce="0,0,0") {
    angularDampingOverride=1
    angularSpringConstant=3000
    bAngularSpring=true
    bOverrideAngularDamping=true
    bUseGravityOverride=true
    boundBone {boneName=dyn_R_tassle_00}
    componentAppliedLinearAccClamp=<1000 1000 1600>
    componentLinearAccScale=<0.5 0.5 0.5>
    componentLinearVelScale=<0.6 0.6 0.6>
    gravityOverride {z=-700}
    physicsBodyDefinitions [{
      boundBone {boneName=dyn_R_tassle_00}
      boxExtents=<10 10 10>
      constraintSetup {angularLimitsMax=<30 0 20>, angularLimitsMin=<-30 -15 -5>, angularTarget=<0 0 0>}
      localJointOffset=<1.4 0 0>
     }]
    simSpaceSettings {maxAngularAcceleration=100, maxAngularVelocity=8, simSpaceAngularAlpha=0.2}
   }
   n7 = Anim Dynamics - Bone: dyn_R_tassle_01(ComponentPose=@n6, Alpha=1, ExternalForce="0,0,0") {
    angularDampingOverride=1
    angularSpringConstant=600
    bAngularSpring=true
    bOverrideAngularDamping=true
    bUseGravityOverride=true
    boundBone {boneName=dyn_R_tassle_01}
    componentAppliedLinearAccClamp=<800 800 1200>
    componentLinearAccScale=<0.5 0.5 0.4>
    componentLinearVelScale=<0.6 0.6 0.5>
    gravityOverride {z=-800}
    physicsBodyDefinitions [{
      boundBone {boneName=dyn_R_tassle_01}
      boxExtents=<12 12 12>
      constraintSetup {angularLimitsMax=<0 30 30>, angularLimitsMin=<0 -60 -30>, angularTarget=<1 0 0>}
      localJointOffset=<1 0 0.4>
     }]
    simSpaceSettings {maxAngularAcceleration=100, maxAngularVelocity=8, simSpaceAngularAlpha=0.2}
   }
   n8 = Anim Dynamics - Bone: dyn_R_tassle_02(ComponentPose=@n7, Alpha=1, ExternalForce="0,0,0") {
    angularDampingOverride=1
    angularSpringConstant=1000
    bAngularSpring=true
    bOverrideAngularDamping=true
    bUseGravityOverride=true
    boundBone {boneName=dyn_R_tassle_02}
    componentAppliedLinearAccClamp=<800 800 1200>
    componentLinearAccScale=<0.5 0.5 0.5>
    componentLinearVelScale=<0.6 0.6 0.6>
    gravityOverride {z=-800}
    physicsBodyDefinitions [{
      boundBone {boneName=dyn_R_tassle_02}
      boxExtents=<10 10 10>
      constraintSetup {angularLimitsMax=<0 30 20>, angularLimitsMin=<0 -60 -20>, angularTarget=<1 0 0>}
      localJointOffset=<1.5 0 0>
     }]
    simSpaceSettings {maxAngularAcceleration=100, maxAngularVelocity=10, simSpaceAngularAlpha=0.2}
   }
   n9 = Anim Dynamics - Bone: dyn_bow_upper(ComponentPose=@n8, Alpha=1, ExternalForce="0,0,0") {
    angularDampingOverride=1
    angularSpringConstant=4500
    bAngularSpring=true
    bOverrideAngularDamping=true
    bUseGravityOverride=true
    boundBone {boneName=dyn_bow_upper}
    componentAppliedLinearAccClamp=<1000 1000 1200>
    componentLinearAccScale=<0.3 0.3 0.3>
    componentLinearVelScale=<0.4 0.4 0.4>
    gravityOverride {z=-980}
    physicsBodyDefinitions [{
      boundBone {boneName=dyn_bow_upper}
      boxExtents=<9 9 9>
      constraintSetup {angularLimitsMax=<0 20 20>, angularLimitsMin=<0 -20 -45>, angularTarget=<1 0 0>}
      localJointOffset=<1 0 0>
     }]
   }
   n10 = Anim Dynamics - Bone: dyn_bow_lower_A_00(ComponentPose=@n9, Alpha=1, ExternalForce="0,0,0") {
    angularDampingOverride=1
    angularSpringConstant=600
    bAngularSpring=true
    bOverrideAngularDamping=true
    bUseGravityOverride=true
    boundBone {boneName=dyn_bow_lower_A_00}
    componentAppliedLinearAccClamp=<1200 1200 1000>
    componentLinearAccScale=<0.4 0.4 0.4>
    componentLinearVelScale=<0.5 0.5 0.5>
    gravityOverride {z=-980}
    physicsBodyDefinitions [{
      boundBone {boneName=dyn_bow_lower_A_00}
      boxExtents=<8 8 8>
      constraintSetup {angularLimitsMax=<0 60 3>, angularLimitsMin=<0 -60 -60>, angularTarget=<0 0 0>}
      localJointOffset=<1.7 0.2 0>
     }]
    simSpaceSettings {maxAngularAcceleration=100, maxAngularVelocity=10, simSpaceAngularAlpha=0.1}
   }
   n11 = Anim Dynamics - Bone: dyn_bow_lower_A_01(ComponentPose=@n10, Alpha=1, ExternalForce="0,0,0") {
    angularDampingOverride=1
    angularSpringConstant=600
    bAngularSpring=true
    bOverrideAngularDamping=true
    bUseGravityOverride=true
    boundBone {boneName=dyn_bow_lower_A_01}
    componentAppliedLinearAccClamp=<1000 1000 1200>
    componentLinearAccScale=<0.4 0.4 0.4>
    componentLinearVelScale=<0.5 0.5 0.5>
    gravityOverride {z=-980}
    physicsBodyDefinitions [{
      boundBone {boneName=dyn_bow_lower_A_01}
      boxExtents=<6 6 6>
      constraintSetup {angularLimitsMax=<5 60 3>, angularLimitsMin=<-5 -60 -70>, angularTarget=<0 0 0>}
      localJointOffset=<1.5 0 0>
     }]
    simSpaceSettings {maxAngularAcceleration=100, maxAngularVelocity=10, simSpaceAngularAlpha=0.1}
   }
   n12 = Anim Dynamics - Bone: dyn_bow_lower_B_00(ComponentPose=@n11, Alpha=1, ExternalForce="0,0,0") {
    angularDampingOverride=1
    angularSpringConstant=700
    bAngularSpring=true
    bOverrideAngularDamping=true
    bUseGravityOverride=true
    boundBone {boneName=dyn_bow_lower_B_00}
    componentAppliedLinearAccClamp=<1200 1200 1200>
    componentLinearAccScale=<0.4 0.4 0.3>
    componentLinearVelScale=<0.5 0.5 0.4>
    gravityOverride {z=-900}
    physicsBodyDefinitions [{
      boundBone {boneName=dyn_bow_lower_B_00}
      boxExtents=<10 10 10>
      constraintSetup {angularLimitsMax=<0 60 3>, angularLimitsMin=<0 -60 -60>, angularTarget=<0 0 0>}
      localJointOffset=<1.6 0.1 0.3>
     }]
    simSpaceSettings {maxAngularAcceleration=100, maxAngularVelocity=10, simSpaceAngularAlpha=0.15}
   }
   n13 = Anim Dynamics - Bone: dyn_bow_lower_B_01(ComponentPose=@n12, Alpha=1, ExternalForce="0,0,0") {
    angularDampingOverride=1
    angularSpringConstant=650
    bAngularSpring=true
    bOverrideAngularDamping=true
    bUseGravityOverride=true
    boundBone {boneName=dyn_bow_lower_B_01}
    componentAppliedLinearAccClamp=<1000 1000 1200>
    componentLinearAccScale=<0.4 0.4 0.4>
    componentLinearVelScale=<0.5 0.5 0.5>
    gravityOverride {z=-900}
    physicsBodyDefinitions [{
      boundBone {boneName=dyn_bow_lower_B_01}
      boxExtents=<7 7 7>
      constraintSetup {angularLimitsMax=<0 60 3>, angularLimitsMin=<0 -60 -70>, angularTarget=<0 0 0>}
      localJointOffset=<1.4 0 0.3>
     }]
    simSpaceSettings {maxAngularAcceleration=100, maxAngularVelocity=10, simSpaceAngularAlpha=0.15}
   }
   n14 = "Transform (Modify) Bone - Bone: dyn_bow_lower_A_00"!(ComponentPose=@n13, bAlphaBoolEnabled:bool=True, Translation="0,0,0", Rotation="0,0,30", Scale="1,1,1") -> (Pose) {
    alphaInputType=Bool
    rotationMode=BMM_Additive
    alphaBoolBlend {blendInTime=0.1, blendOutTime=0.3}
    boneToModify {boneName=dyn_bow_lower_A_00}
    rotation {roll=30}
   }
   n15 = "Transform (Modify) Bone - Bone: dyn_bow_lower_A_01"!(ComponentPose=@n14, bAlphaBoolEnabled:bool=True, Translation="0,0,0", Rotation="18.877,0.000,53.956", Scale="1,1,1") -> (Pose) {
    alphaInputType=Bool
    rotationMode=BMM_Additive
    alphaBoolBlend {blendInTime=0.1, blendOutTime=0.3}
    boneToModify {boneName=dyn_bow_lower_A_01}
    rotation=<rot 18.877 0 53.956>
   }
   n16 = "Transform (Modify) Bone - Bone: dyn_bow_lower_B_00"!(ComponentPose=@n15, bAlphaBoolEnabled:bool=True, Translation="0,0,0", Rotation="0,0,30", Scale="1,1,1") -> (Pose) {
    alphaInputType=Bool
    rotationMode=BMM_Additive
    alphaBoolBlend {blendInTime=0.1, blendOutTime=0.3}
    boneToModify {boneName=dyn_bow_lower_B_00}
    rotation {roll=30}
   }
   n17 = "Transform (Modify) Bone - Bone: dyn_bow_lower_B_01"!(ComponentPose=@n16, bAlphaBoolEnabled:bool=True, Translation="0,0,0", Rotation="18.877,0.000,53.956", Scale="1,1,1") -> (Pose) {
    alphaInputType=Bool
    rotationMode=BMM_Additive
    alphaBoolBlend {blendInTime=0.1, blendOutTime=0.3}
    boneToModify {boneName=dyn_bow_lower_B_01}
    rotation {pitch=18.877, roll=53.956}
   }
   n18 = "Transform (Modify) Bone - Bone: dyn_L_tassle_00"!(ComponentPose=@n17, bAlphaBoolEnabled:bool=True, Translation="0,0,0", Rotation="6.828,-63.871,-96.496", Scale="1,1,1") -> (Pose) {
    alphaInputType=Bool
    rotationMode=BMM_Additive
    alphaBoolBlend {blendInTime=0.1, blendOutTime=0.3}
    boneToModify {boneName=dyn_L_tassle_00}
    rotation=<rot 6.828 -63.871 -96.496>
   }
   n19 = "Transform (Modify) Bone - Bone: dyn_L_tassle_01"!(ComponentPose=@n18, bAlphaBoolEnabled:bool=True, Translation="0,0,0", Rotation="-11.412,-42.190,10.169", Scale="1,1,1") -> (Pose) {
    alphaInputType=Bool
    rotationMode=BMM_Additive
    alphaBoolBlend {blendInTime=0.1, blendOutTime=0.3}
    boneToModify {boneName=dyn_L_tassle_01}
    rotation=<rot -11.412 -42.19 10.169>
   }
   n20 = "Transform (Modify) Bone - Bone: dyn_R_tassle_00"!(ComponentPose=@n19, bAlphaBoolEnabled:bool=True, Translation="0,0,0", Rotation="-54.476,-18.274,76.062", Scale="1,1,1") -> (Pose) {
    alphaInputType=Bool
    rotationMode=BMM_Additive
    alphaBoolBlend {blendInTime=0.1, blendOutTime=0.3}
    boneToModify {boneName=dyn_R_tassle_00}
    rotation=<rot -54.476 -18.274 76.062>
   }
   n21 = "Transform (Modify) Bone - Bone: dyn_R_tassle_01"!(ComponentPose=@n20, bAlphaBoolEnabled:bool=True, Translation="0,0,0", Rotation="-0.000,40.423,-31.670", Scale="1,1,1") -> (Pose) {
    alphaInputType=Bool
    rotationMode=BMM_Additive
    alphaBoolBlend {blendInTime=0.1, blendOutTime=0.3}
    boneToModify {boneName=dyn_R_tassle_01}
    rotation=<rot 0 40.423 -31.67>
   }
   Output Pose(Result=Component To Local(ComponentPose=@n21)) {layerGroup=None}
  ]}
 EventGraph {nodes [
   Event BlueprintUpdateAnimation()
   > n1 = Sequence!(execute) -> (then_0, then_1, then_2:Exec, then_3:Exec)
   > n2 = Cast To FortPlayerAnimInstance(Object=GetAnimInstance(self=GetOwnerSkeletalMeshComponent()))
   > Set isPickaxe(isPickaxe:bool=false) -> (Output_Get:bool)
   > n3 = Switch on EFortWeaponCoreAnimation!(execute, Selection:EFortWeaponCoreAnimation Enum=Get WeaponCoreAnim(self=@n2)) -> (
    Melee:Exec
    Pistol:Exec
    Shotgun:Exec
    PaperBlueprint:Exec
    Rifle:Exec
    MeleeOneHand:Exec
    MachinePistol:Exec
    RocketLauncher:Exec
    GrenadeLauncher:Exec
    GoingCommando:Exec
    AssaultRifle:Exec
    TacticalShotgun:Exec
    SniperRifle:Exec
    TrapPlacement:Exec
    ShoulderLauncher:Exec
    AbilityDecoTool:Exec
    Crossbow:Exec
    C4:Exec
    RemoteControl:Exec
    DualWield:Exec
    AR_BullPup:Exec
    AR_ForwardGrip:Exec
    MedPackPaddles:Exec
    SMG_P90:Exec
    AR_DrumGun:Exec
    Consumable_Small:Exec
    Consumable_Large:Exec
    Balloon:Exec
    MountedTurret:Exec
    CreativeTool:Exec
    ExplosiveBow:Exec
    AshtonIndigo:Exec
    AshtonChicago:Exec
    MeleeDualWield:Exec
    Unarmed:Exec
   )
   Set isPickaxe(execute=[@n3.DualWield, @n3.Melee, @n3.MeleeDualWield, @n3.MeleeOneHand], isPickaxe=true)
   Set isInAir(execute=@n1.then_1, isInAir:bool=AND Boolean(A=Get bIsSkydiveFloating(), B=Get bIsSkyDiving())) -> (Output_Get:bool)
   Event OnCharacterCustomizationLoaded()
   > n4 = Branch(Condition=PickaxeItemDefinitionHasAnyMetaTag(Tags=Make Array("[0]"=(TagName="Cosmetics.MetaTags.DragonPry")), bRequiresEquipped=false))
   > "Set DragonPry  pickaxe"("DragonPry  pickaxe":bool=true) -> (Output_Get:bool)
   "Set DragonPry  pickaxe"(execute=@n4.else, "DragonPry  pickaxe"=false)
   TryGetPawnOwner()
  ]}
]
variables [
 "DragonPry  pickaxe" {type=boolean}
 holdingDragonPry {type=boolean}
 isInAir {type=boolean}
 isPickaxe {type=boolean}
]
